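# Physics asset for the back poncho of the female medium Bounty Hunter body, setting up constrained physics for the hanging garment.
# Seven thin boxes, narrowing down the chain, link from spine_05 through dyn_back_poncho_01 to 07.
# Joints from 03 down lock twist, with 03 to 05 also locking swing1 so the middle bends on one axis.
# Capsules on pelvis, spine, thighs and calves form the body shapes, joined by free angular constraints.
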PhysicsAsset /BRCosmetics/Characters/Player/Female/Medium/Bodies/F_MED_BountyHunter/Meshes/F_MED_BountyHunter_BackPoncho_Constrained_Physics
bodies {
 calf_l [{shapeType=Capsule, length=30, radius=7, center=<-20 0 0>, rotation=<rot 90 0 0>}]
 calf_r [{shapeType=Capsule, length=30, radius=7, center=<20 0 0>, rotation=<rot 90 0 0>}]
 dyn_back_poncho_01 [{shapeType=Box, extentX=5, extentY=1.5, extentZ=25.3509, center=<2.5 0 0.6562>}]
 dyn_back_poncho_02 [{shapeType=Box, extentX=5, extentY=1.5, extentZ=23.1738, center=<2.5 0 0>}]
 dyn_back_poncho_03 [{shapeType=Box, extentX=5, extentY=1.5, extentZ=19.1977, center=<2.5 -0.0002 -2.4008>}]
 dyn_back_poncho_04 [{shapeType=Box, extentX=5, extentY=1.5, extentZ=16.0604, center=<2.5 0 -4.199>}]
 dyn_back_poncho_05 [{shapeType=Box, extentX=5, extentY=1.5, extentZ=12, center=<2.5 -0.0001 -6.3814>}]
 dyn_back_poncho_06 [{shapeType=Box, extentX=5, extentY=1.5, extentZ=10.5829, center=<2.5 -0.0001 -8.2524>}]
 dyn_back_poncho_07 [{shapeType=Box, extentX=5, extentY=1.5, extentZ=5.8017, center=<2.5 -0.0001 -11.4091>}]
 pelvis [{shapeType=Capsule, length=13.1181, radius=11.5416, center=<-3.7041 -1.1856 -0.0002>}]
 spine_01 [{shapeType=Capsule, length=10.8484, radius=7.5559}]
 spine_02 [{shapeType=Capsule, length=11.7806, radius=10.6087}]
 spine_03 [{shapeType=Capsule, length=0, radius=10.101}]
 spine_04 [{shapeType=Capsule, length=17.7468, radius=9.6449}]
 spine_05 [{shapeType=Capsule, length=13.6593, radius=10.086, center=<-2.0838 0.1524 0.0001>}]
 thigh_l [{shapeType=Capsule, length=28, radius=9, center=<-20 0 0>, rotation=<rot 90 0 0>}]
 thigh_r [{shapeType=Capsule, length=28, radius=9, center=<20 0 0>, rotation=<rot 90 0 0>}]
}
constraints [
 {
  bone1Name=calf_l
  bone2Name=thigh_l
  swing1LimitDegrees=45
  swing1Motion=Free
  swing2LimitDegrees=45
  swing2Motion=Free
  twistLimitDegrees=45
  twistMotion=Free
 }
 {
  bone1Name=calf_r
  bone2Name=thigh_r
  swing1LimitDegrees=45
  swing1Motion=Free
  swing2LimitDegrees=45
  swing2Motion=Free
  twistLimitDegrees=45
  twistMotion=Free
 }
 {
  bone1Name=thigh_l
  bone2Name=pelvis
  swing1LimitDegrees=45
  swing1Motion=Free
  swing2LimitDegrees=45
  swing2Motion=Free
  twistLimitDegrees=45
  twistMotion=Free
 }
 {
  bone1Name=thigh_r
  bone2Name=pelvis
  swing1LimitDegrees=45
  swing1Motion=Free
  swing2LimitDegrees=45
  swing2Motion=Free
  twistLimitDegrees=45
  twistMotion=Free
 }
 {
  bone1Name=spine_01
  bone2Name=pelvis
  swing1LimitDegrees=45
  swing1Motion=Free
  swing2LimitDegrees=45
  swing2Motion=Free
  twistLimitDegrees=45
  twistMotion=Free
 }
 {
  bone1Name=spine_02
  bone2Name=spine_01
  swing1LimitDegrees=45
  swing1Motion=Free
  swing2LimitDegrees=45
  swing2Motion=Free
  twistLimitDegrees=45
  twistMotion=Free
 }
 {
  bone1Name=spine_03
  bone2Name=spine_02
  swing1LimitDegrees=45
  swing1Motion=Free
  swing2LimitDegrees=45
  swing2Motion=Free
  twistLimitDegrees=45
  twistMotion=Free
 }
 {
  bone1Name=spine_04
  bone2Name=spine_03
  swing1LimitDegrees=45
  swing1Motion=Free
  swing2LimitDegrees=45
  swing2Motion=Free
  twistLimitDegrees=45
  twistMotion=Free
 }
 {
  bone1Name=spine_05
  bone2Name=spine_04
  swing1LimitDegrees=45
  swing1Motion=Free
  swing2LimitDegrees=45
  swing2Motion=Free
  twistLimitDegrees=45
  twistMotion=Free
 }
 {bone1Name=dyn_back_poncho_01, bone2Name=spine_05, swing1LimitDegrees=45, swing2LimitDegrees=45, twistLimitDegrees=5}
 {bone1Name=dyn_back_poncho_02, bone2Name=dyn_back_poncho_01, swing1LimitDegrees=45, swing2LimitDegrees=45, twistLimitDegrees=5}
 {
  bone1Name=dyn_back_poncho_03
  bone2Name=dyn_back_poncho_02
  swing1LimitDegrees=45
  swing1Motion=Locked
  swing2LimitDegrees=45
  twistLimitDegrees=5
  twistMotion=Locked
 }
 {
  bone1Name=dyn_back_poncho_04
  bone2Name=dyn_back_poncho_03
  swing1LimitDegrees=0
  swing1Motion=Locked
  swing2LimitDegrees=45
  twistLimitDegrees=0
  twistMotion=Locked
 }
 {
  bone1Name=dyn_back_poncho_05
  bone2Name=dyn_back_poncho_04
  swing1LimitDegrees=45
  swing1Motion=Locked
  swing2LimitDegrees=45
  twistLimitDegrees=5
  twistMotion=Locked
 }
 {
  bone1Name=dyn_back_poncho_06
  bone2Name=dyn_back_poncho_05
  swing1LimitDegrees=20
  swing2LimitDegrees=45
  twistLimitDegrees=5
  twistMotion=Locked
 }
 {
  bone1Name=dyn_back_poncho_07
  bone2Name=dyn_back_poncho_06
  swing1LimitDegrees=45
  swing2LimitDegrees=45
  twistLimitDegrees=5
  twistMotion=Locked
 }
]
