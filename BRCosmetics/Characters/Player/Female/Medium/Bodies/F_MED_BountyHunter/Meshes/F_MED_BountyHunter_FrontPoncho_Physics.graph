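# Physics asset for the front poncho of the female medium Bounty Hunter body, giving its dynamic bones collision shapes and joints.
# Five box bodies form a chain from dyn_front_poncho_01 to 05, constrained in sequence and anchored to spine_04.
# Capsules on the pelvis, spine, thighs and calves approximate the wearer's body.
# Those body capsules are linked by joints with free swing and twist, unlike the poncho chain.
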PhysicsAsset /BRCosmetics/Characters/Player/Female/Medium/Bodies/F_MED_BountyHunter/Meshes/F_MED_BountyHunter_FrontPoncho_Physics
bodies {
 calf_l [{shapeType=Capsule, length=30, radius=7, center=<-20 0 0>, rotation=<rot 90 0 0>}]
 calf_r [{shapeType=Capsule, length=30, radius=7, center=<20 0 0>, rotation=<rot 90 0 0>}]
 dyn_front_poncho_01 [{shapeType=Box, extentX=6, extentY=1.5, extentZ=12, center=<3 0 0>}]
 dyn_front_poncho_02 [{shapeType=Box, extentX=6, extentY=1.5, extentZ=12, center=<3 -0.0001 -4.9702>}]
 dyn_front_poncho_03 [{shapeType=Box, extentX=6, extentY=1.5, extentZ=12, center=<3 -0.0002 -7.8479>}]
 dyn_front_poncho_04 [{shapeType=Box, extentX=6, extentY=1.5, extentZ=11.5, center=<3 0 -10.6816>}]
 dyn_front_poncho_05 [{shapeType=Box, extentX=6, extentY=1.5, extentZ=8.5, center=<3 0.0001 -13.7011>}]
 pelvis [{shapeType=Capsule, length=16.6069, radius=13.4126, center=<-4.4263 -1.8043 -0.0002>}]
 spine_01 [{shapeType=Capsule, length=0.505, radius=0.505}]
 spine_02 [{shapeType=Capsule, length=11.7806, radius=10.6087}]
 spine_03 [{shapeType=Capsule, length=5.0116, radius=2.101}]
 spine_04 [{shapeType=Capsule, length=13.768, radius=13.6237}]
 spine_05 [{shapeType=Capsule, length=11.9745, radius=11.7708}]
 thigh_l [{shapeType=Capsule, length=28, radius=9, center=<-20 0 0>, rotation=<rot 90 0 0>}]
 thigh_r [{shapeType=Capsule, length=28, radius=9, center=<20 0 0>, rotation=<rot 90 0 0>}]
}
constraints [
 {
  bone1Name=calf_l
  bone2Name=thigh_l
  swing1LimitDegrees=45
  swing1Motion=Free
  swing2LimitDegrees=45
  swing2Motion=Free
  twistLimitDegrees=45
  twistMotion=Free
 }
 {
  bone1Name=calf_r
  bone2Name=thigh_r
  swing1LimitDegrees=45
  swing1Motion=Free
  swing2LimitDegrees=45
  swing2Motion=Free
  twistLimitDegrees=45
  twistMotion=Free
 }
 {
  bone1Name=thigh_l
  bone2Name=pelvis
  swing1LimitDegrees=45
  swing1Motion=Free
  swing2LimitDegrees=45
  swing2Motion=Free
  twistLimitDegrees=45
  twistMotion=Free
 }
 {
  bone1Name=thigh_r
  bone2Name=pelvis
  swing1LimitDegrees=45
  swing1Motion=Free
  swing2LimitDegrees=45
  swing2Motion=Free
  twistLimitDegrees=45
  twistMotion=Free
 }
 {
  bone1Name=spine_01
  bone2Name=pelvis
  swing1LimitDegrees=45
  swing1Motion=Free
  swing2LimitDegrees=45
  swing2Motion=Free
  twistLimitDegrees=45
  twistMotion=Free
 }
 {
  bone1Name=spine_02
  bone2Name=spine_01
  swing1LimitDegrees=45
  swing1Motion=Free
  swing2LimitDegrees=45
  swing2Motion=Free
  twistLimitDegrees=45
  twistMotion=Free
 }
 {
  bone1Name=spine_03
  bone2Name=spine_02
  swing1LimitDegrees=45
  swing1Motion=Free
  swing2LimitDegrees=45
  swing2Motion=Free
  twistLimitDegrees=45
  twistMotion=Free
 }
 {
  bone1Name=spine_04
  bone2Name=spine_03
  swing1LimitDegrees=45
  swing1Motion=Free
  swing2LimitDegrees=45
  swing2Motion=Free
  twistLimitDegrees=45
  twistMotion=Free
 }
 {
  bone1Name=spine_05
  bone2Name=spine_04
  swing1LimitDegrees=45
  swing1Motion=Free
  swing2LimitDegrees=45
  swing2Motion=Free
  twistLimitDegrees=45
  twistMotion=Free
 }
 {bone1Name=dyn_front_poncho_01, bone2Name=spine_04, swing1LimitDegrees=45, swing2LimitDegrees=45, twistLimitDegrees=45}
 {bone1Name=dyn_front_poncho_02, bone2Name=dyn_front_poncho_01, swing1LimitDegrees=45, swing2LimitDegrees=45, twistLimitDegrees=45}
 {bone1Name=dyn_front_poncho_03, bone2Name=dyn_front_poncho_02, swing1LimitDegrees=45, swing2LimitDegrees=45, twistLimitDegrees=45}
 {bone1Name=dyn_front_poncho_04, bone2Name=dyn_front_poncho_03, swing1LimitDegrees=45, swing2LimitDegrees=45, twistLimitDegrees=45}
 {bone1Name=dyn_front_poncho_05, bone2Name=dyn_front_poncho_04, swing1LimitDegrees=45, swing2LimitDegrees=45, twistLimitDegrees=45}
]
