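# Physics asset for the rope on the female medium Bounty Hunter body, covering only the right leg and the rope's dyn bones.
# Capsules on the right thigh and calf, joined by an angularly locked constraint, give the leg its shape.
# Rope strands A, B and C are constrained to the thigh, with C continuing as a two-bone chain.
# All rope joints share the same narrow swing and tighter twist limit angles.
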PhysicsAsset /BRCosmetics/Characters/Player/Female/Medium/Bodies/F_MED_BountyHunter/Meshes/F_MED_BountyHunter_Rope_Physics
bodies {
 calf_r [{
   shapeType=Capsule
   length=25.5716
   radius=7.8442
   shapeName=calf_r_capsule
   center=<15.3753 3.2591 0.0252>
   rotation=<rot -0.7078 -87.937 -94.0178>
  }]
 dyn_rope_A_01 [{
   shapeType=Capsule
   length=6
   radius=5.75
   shapeName=dyn_rope_A_01_capsule
   center=<4.9733 -1.6483 -0.2744>
   rotation=<rot 0 89.2822 -88.8041>
  }]
 dyn_rope_B_01 [{shapeType=Capsule, length=6.25, radius=6, shapeName=dyn_rope_B_01_capsule, center=<5.1655 -1.4243 0.6351>, rotation=<rot 0 86.3077 -88.2344>}]
 dyn_rope_C_01 [{
   shapeType=Capsule
   length=2.25
   radius=5.75
   shapeName=dyn_rope_C_01_capsule
   center=<3.312 0.0024 0.0366>
   rotation=<rot 4.2596 -89.7522 -91.5531>
  }]
 dyn_rope_C_02 [{
   shapeType=Capsule
   length=1.5
   radius=6.75
   shapeName=dyn_rope_C_02_capsule
   center=<2.3818 -0.2239 -0.5462>
   rotation=<rot -3.1843 96.3509 -90.0002>
  }]
 thigh_r [{
   shapeType=Capsule
   length=25.0892
   radius=7.9431
   shapeName=thigh_r_capsule
   center=<23.9505 0.2999 -1.4053>
   rotation=<rot 0.0002 -80.7483 -90.8391>
  }]
}
constraints [
 {
  bone1Name=calf_r
  bone2Name=thigh_r
  swing1LimitDegrees=45
  swing1Motion=Locked
  swing2LimitDegrees=45
  swing2Motion=Locked
  twistLimitDegrees=45
  twistMotion=Locked
 }
 {bone1Name=dyn_rope_A_01, bone2Name=thigh_r, swing1LimitDegrees=20, swing2LimitDegrees=20, twistLimitDegrees=5}
 {bone1Name=dyn_rope_B_01, bone2Name=thigh_r, swing1LimitDegrees=20, swing2LimitDegrees=20, twistLimitDegrees=5}
 {bone1Name=dyn_rope_C_01, bone2Name=thigh_r, swing1LimitDegrees=20, swing2LimitDegrees=20, twistLimitDegrees=5}
 {bone1Name=dyn_rope_C_02, bone2Name=dyn_rope_C_01, swing1LimitDegrees=20, swing2LimitDegrees=20, twistLimitDegrees=5}
]
