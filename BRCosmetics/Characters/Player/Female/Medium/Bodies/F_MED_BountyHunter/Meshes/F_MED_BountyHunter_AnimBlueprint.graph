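# Post-process style AnimBlueprint for the female BountyHunter cosmetic body, copying the parent mesh pose and adding rope and poncho motion.
# A constrained, simulated rope feeds separate back and front poncho branches, merged by per-bone layering.
# Each switches from RigidBody dynamics, pushed by skydiving force, to weighted bone constraints when physics is disabled or LOD exceeds 1.
# An equipped backpack also switches the back, whose constrained branch retains limited RigidBody simulation and LOD offsets.
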AnimBlueprint /BRCosmetics/Characters/Player/Female/Medium/Bodies/F_MED_BountyHunter/Meshes/F_MED_BountyHunter_AnimBlueprint
target_skeleton=&/F_MED_BountyHunter_Skeleton
graphs [
 AnimGraph {nodes [
   n1 = Anim Dynamics - Bone: dyn_rope_A_01(Alpha=1, ExternalForce="0,0,0") {
    angularDampingOverride=1
    angularSpringConstant=1500
    bAngularSpring=true
    bGravityOverrideInSimSpace=true
    bOverrideAngularDamping=true
    gravityScale=0.35
    simulationSpace=RootRelative
    alphaBoolBlend {blendOption=Linear}
    boundBone {boneName=dyn_rope_A_01}
    physicsBodyDefinitions [{
      sphereCollisionRadius=0
      boundBone {boneName=dyn_rope_A_01}
      boxExtents=<10 20 20>
      constraintSetup {angularLimitsMax=<0 30 0>, angularLimitsMin=<0 -30 -40>, angularTarget=<0 0 0>}
      localJointOffset=<10 0 0>
     }]
   }
   n2 = Anim Dynamics - Bone: dyn_rope_B_01(ComponentPose=@n1, Alpha=1, ExternalForce="0,0,0") {
    angularDampingOverride=1
    angularSpringConstant=1500
    bAngularSpring=true
    bGravityOverrideInSimSpace=true
    bOverrideAngularDamping=true
    gravityScale=0.35
    simulationSpace=RootRelative
    alphaBoolBlend {blendOption=Linear}
    boundBone {boneName=dyn_rope_B_01}
    physicsBodyDefinitions [{
      sphereCollisionRadius=0
      boundBone {boneName=dyn_rope_B_01}
      boxExtents=<10 20 20>
      constraintSetup {angularLimitsMax=<0 30 0>, angularLimitsMin=<0 -30 -40>, angularTarget=<0 0 0>}
      localJointOffset=<10 0 0>
     }]
   }
   n3 = Anim Dynamics - Bone: dyn_rope_C_01(ComponentPose=@n2, Alpha=1, ExternalForce="0,0,0") {
    angularDampingOverride=1
    angularSpringConstant=1500
    bAngularSpring=true
    bGravityOverrideInSimSpace=true
    bOverrideAngularDamping=true
    gravityScale=0.35
    simulationSpace=RootRelative
    alphaBoolBlend {blendOption=Linear}
    boundBone {boneName=dyn_rope_C_01}
    physicsBodyDefinitions [{
      sphereCollisionRadius=0
      boundBone {boneName=dyn_rope_C_01}
      boxExtents=<10 20 20>
      constraintSetup {angularLimitsMax=<0 30 0>, angularLimitsMin=<0 -30 -40>, angularTarget=<0 0 0>}
      localJointOffset=<10 0 0>
     }]
   }
   Anim Dynamics - Bone: dyn_rope_C_02(ComponentPose=@n3, Alpha=1, ExternalForce="0,0,0") {
    angularDampingOverride=1
    angularSpringConstant=1500
    bAngularSpring=true
    bGravityOverrideInSimSpace=true
    bOverrideAngularDamping=true
    gravityScale=0.35
    simulationSpace=RootRelative
    alphaBoolBlend {blendOption=Linear}
    boundBone {boneName=dyn_rope_C_02}
    physicsBodyDefinitions [{
      sphereCollisionRadius=0
      boundBone {boneName=dyn_rope_C_02}
      boxExtents=<10 20 20>
      constraintSetup {angularLimitsMax=<0 30 0>, angularLimitsMin=<0 -30 -40>, angularTarget=<0 0 0>}
      localJointOffset=<10 0 0>
     }]
   }
   Output Pose(Result=Layered blend per bone(BasePose=Use cached pose 'front_cache'(), BlendPoses_0=Use cached pose 'back_cache'()) {layerSetup [{branchFilters [{boneName=dyn_back_poncho_01}]}]}) {layerGroup=None}
   n4 = "Constraint - dyn_rope_master (PP)"!(
    ComponentPose=Local To Component(LocalPose=Copy Pose From Mesh() {bUseAttachedParent=true})
    Alpha=1
    ConstraintWeights_0=0.8863
    ConstraintWeights_1:float=0.1137
   ) -> (Pose) {
    alphaBoolBlend {blendOption=Linear}
    boneToModify {boneName=dyn_rope_master}
    constraintSetup [{targetBone {boneName=thigh_r}}, {targetBone {boneName=thigh_twist_01_r}}]
    constraintWeights [1,1]
   }
   n5 = RigidBody(ComponentPose=@n4) {
    bOverrideWorldGravity=true
    overridePhysicsAsset=&/F_MED_BountyHunter_Rope_Physics
    simulationSpace=BaseBoneSpace
    alphaBoolBlend {blendOption=Linear}
    overrideWorldGravity {z=-980}
   }
   Save cached pose 'copypose_cache'(Pose=Component To Local(ComponentPose=@n5)) {cacheName=copypose_cache}
   n6 = "Constraint - dyn_back_poncho_01 (PPPP)"!(
    ComponentPose=Local To Component(LocalPose=Use cached pose 'copypose_cache'())
    Alpha=1
    ConstraintWeights_0=0.1373
    ConstraintWeights_1:float=0.0196
    ConstraintWeights_2:float=0.7059
    ConstraintWeights_3:float=0.1373
   ) -> (Pose) {
    alphaBoolBlend {blendOption=Linear}
    boneToModify {boneName=dyn_back_poncho_01}
    constraintSetup [{targetBone {boneName=clavicle_r}}, {targetBone {boneName=spine_04}}, {targetBone {boneName=spine_05}}, {targetBone {boneName=clavicle_l}}]
    constraintWeights [1,1,1,1]
   }
   n7 = "Constraint - dyn_back_poncho_02 (PPPP)"!(
    ComponentPose=@n6
    Alpha=1
    ConstraintWeights_0=0.1059
    ConstraintWeights_1:float=0.3059
    ConstraintWeights_2:float=0.4824
    ConstraintWeights_3:float=0.1059
   ) -> (Pose) {
    alphaBoolBlend {blendOption=Linear}
    boneToModify {boneName=dyn_back_poncho_02}
    constraintSetup [{targetBone {boneName=clavicle_r}}, {targetBone {boneName=spine_04}}, {targetBone {boneName=spine_05}}, {targetBone {boneName=clavicle_l}}]
    constraintWeights [1,1,1,1]
   }
   n8 = "Constraint - dyn_back_poncho_03 (PPPPPPP)"!(
    ComponentPose=@n7
    Alpha=1
    ConstraintWeights_0=0.0549
    ConstraintWeights_1:float=0.0157
    ConstraintWeights_2:float=0.0431
    ConstraintWeights_3:float=0.0667
    ConstraintWeights_4:float=0.5765
    ConstraintWeights_5:float=0.1882
    ConstraintWeights_6:float=0.0549
   ) -> (Pose) {
    alphaBoolBlend {blendOption=Linear}
    boneToModify {boneName=dyn_back_poncho_03}
    constraintSetup [
     {targetBone {boneName=clavicle_l}}
     {targetBone {boneName=spine_01}}
     {targetBone {boneName=spine_02}}
     {targetBone {boneName=spine_03}}
     {targetBone {boneName=spine_04}}
     {targetBone {boneName=spine_05}}
     {targetBone {boneName=clavicle_r}}
    ]
    constraintWeights [1,1,1,1,1,1,1]
   }
   n9 = "Constraint - dyn_back_poncho_04 (PPPPPPP)"!(
    ComponentPose=@n8
    Alpha=1
    ConstraintWeights_0=0.0168
    ConstraintWeights_1:float=0.0751
    ConstraintWeights_2:float=0.1961
    ConstraintWeights_3:float=0.2751
    ConstraintWeights_4:float=0.3653
    ConstraintWeights_5:float=0.0549
    ConstraintWeights_6:float=0.0168
   ) -> (Pose) {
    alphaBoolBlend {blendOption=Linear}
    boneToModify {boneName=dyn_back_poncho_04}
    constraintSetup [
     {targetBone {boneName=clavicle_l}}
     {targetBone {boneName=spine_01}}
     {targetBone {boneName=spine_02}}
     {targetBone {boneName=spine_03}}
     {targetBone {boneName=spine_04}}
     {targetBone {boneName=spine_05}}
     {targetBone {boneName=clavicle_r}}
    ]
    constraintWeights [1,1,1,1,1,1,1]
   }
   n10 = "Constraint - dyn_back_poncho_05 (PPPPPPPP)"!(
    ComponentPose=@n9
    Alpha=0
    ConstraintWeights_0=0.0002
    ConstraintWeights_1:float=0.1499
    ConstraintWeights_2:float=0.3253
    ConstraintWeights_3:float=0.3086
    ConstraintWeights_4:float=0.1637
    ConstraintWeights_5:float=0.0067
    ConstraintWeights_6:float=0
    ConstraintWeights_7:float=0.0456
   ) -> (Pose) {
    alphaBoolBlend {blendOption=Linear}
    boneToModify {boneName=dyn_back_poncho_05}
    constraintSetup [
     {targetBone {boneName=clavicle_l}}
     {targetBone {boneName=spine_01}}
     {targetBone {boneName=spine_02}}
     {targetBone {boneName=spine_03}}
     {targetBone {boneName=spine_04}}
     {targetBone {boneName=spine_05}}
     {targetBone {boneName=clavicle_r}}
     {targetBone {boneName=pelvis}}
    ]
    constraintWeights [1,1,1,1,1,1,1,1]
   }
   n11 = "Constraint - dyn_back_poncho_06 (PPPPPPPP)"!(
    ComponentPose=@n10
    Alpha=0
    ConstraintWeights_0=0
    ConstraintWeights_1:float=0.2748
    ConstraintWeights_2:float=0.2953
    ConstraintWeights_3:float=0.0548
    ConstraintWeights_4:float=0.0032
    ConstraintWeights_5:float=0
    ConstraintWeights_6:float=0
    ConstraintWeights_7:float=0.3718
   ) -> (Pose) {
    alphaBoolBlend {blendOption=Linear}
    boneToModify {boneName=dyn_back_poncho_06}
    constraintSetup [
     {targetBone {boneName=clavicle_l}}
     {targetBone {boneName=spine_01}}
     {targetBone {boneName=spine_02}}
     {targetBone {boneName=spine_03}}
     {targetBone {boneName=spine_04}}
     {targetBone {boneName=spine_05}}
     {targetBone {boneName=clavicle_r}}
     {targetBone {boneName=pelvis}}
    ]
    constraintWeights [1,1,1,1,1,1,1,1]
   }
   n12 = "Constraint - dyn_back_poncho_07 (P)"!(ComponentPose=@n11, Alpha=0, ConstraintWeights_0=1, ConstraintWeights_1:float=0) -> (Pose) {
    alphaBoolBlend {blendOption=Linear}
    boneToModify {boneName=dyn_back_poncho_07}
    constraintSetup [{targetBone {boneName=pelvis}}]
    constraintWeights [1,0]
   }
   n13 = RigidBody(ComponentPose=@n12) {
    bClampLinearTranslationLimitToRefPose=true
    bOverrideWorldGravity=true
    overridePhysicsAsset=&/F_MED_BountyHunter_BackPoncho_Constrained_Physics
    simulationSpace=BaseBoneSpace
    alphaBoolBlend {blendOption=Linear}
    baseBoneRef {boneName=pelvis}
    componentLinearAccScale=<0.4 0.4 0.6>
    componentLinearVelScale=<0.4 0.4 0.6>
    overrideWorldGravity {z=-980}
   }
   n14 = "Transform (Modify) Bone - Bone: dyn_back_poncho_01"(
    ComponentPose=@n13
    Alpha=1
    Translation=Get back_poncho_lod_offset() -> (back_poncho_lod_offset:Vector)
    Rotation=Get back_poncho_lod_rot_offset() -> (back_poncho_lod_rot_offset:Rotator)
    Scale="1,1,1"
   ) {rotationMode=BMM_Additive, translationMode=BMM_Additive, alphaBoolBlend {blendOption=Linear}, boneToModify {boneName=dyn_back_poncho_01}}
   Save cached pose 'back_constraint_cache'(Pose=Component To Local(ComponentPose=@n14)) {cacheName=back_constraint_cache}
   n15 = "Constraint - dyn_front_poncho_01 (PP)"!(
    ComponentPose=Local To Component(LocalPose=Use cached pose 'copypose_cache'())
    Alpha=1
    ConstraintWeights_0=0.9647
    ConstraintWeights_1:float=0.0353
   ) -> (Pose) {
    alphaBoolBlend {blendOption=Linear}
    boneToModify {boneName=dyn_front_poncho_01}
    constraintSetup [{targetBone {boneName=spine_04}}, {targetBone {boneName=spine_05}}]
    constraintWeights [1,1]
   }
   n16 = "Constraint - dyn_front_poncho_02 (P)"!(ComponentPose=@n15, Alpha=1, ConstraintWeights_0=1, ConstraintWeights_1:float=0) -> (Pose) {
    alphaBoolBlend {blendOption=Linear}
    boneToModify {boneName=dyn_front_poncho_02}
    constraintSetup [{targetBone {boneName=spine_04}}]
    constraintWeights [1,0]
   }
   n17 = "Constraint - dyn_front_poncho_03 (PPPPPP)"!(
    ComponentPose=@n16
    Alpha=1
    ConstraintWeights_0=0.0934
    ConstraintWeights_1:float=0.2064
    ConstraintWeights_2:float=0.2451
    ConstraintWeights_3:float=0.4521
    ConstraintWeights_4:float=0
    ConstraintWeights_5:float=0.003
   ) -> (Pose) {
    alphaBoolBlend {blendOption=Linear}
    boneToModify {boneName=dyn_front_poncho_03}
    constraintSetup [
     {targetBone {boneName=spine_01}}
     {targetBone {boneName=spine_02}}
     {targetBone {boneName=spine_03}}
     {targetBone {boneName=spine_04}}
     {targetBone {boneName=spine_05}}
     {targetBone {boneName=pelvis}}
    ]
    constraintWeights [1,1,1,1,1,1]
   }
   n18 = "Constraint - dyn_front_poncho_04 (PPPPPPPP)"!(
    ComponentPose=@n17
    Alpha=1
    ConstraintWeights_0=0
    ConstraintWeights_1:float=0.2281
    ConstraintWeights_2:float=0.4042
    ConstraintWeights_3:float=0.0437
    ConstraintWeights_4:float=0.1267
    ConstraintWeights_5:float=0
    ConstraintWeights_6:float=0
    ConstraintWeights_7:float=0.1972
   ) -> (Pose) {
    alphaBoolBlend {blendOption=Linear}
    boneToModify {boneName=dyn_front_poncho_04}
    constraintSetup [
     {targetBone {boneName=clavicle_l}}
     {targetBone {boneName=spine_01}}
     {targetBone {boneName=spine_02}}
     {targetBone {boneName=spine_03}}
     {targetBone {boneName=spine_04}}
     {targetBone {boneName=spine_05}}
     {targetBone {boneName=clavicle_r}}
     {targetBone {boneName=pelvis}}
    ]
    constraintWeights [1,1,1,1,1,1,1,1]
   }
   n19 = "Constraint - dyn_front_poncho_05 (P)"!(ComponentPose=@n18, Alpha=1, ConstraintWeights_0=1, ConstraintWeights_1:float=0) -> (Pose) {
    alphaBoolBlend {blendOption=Linear}
    boneToModify {boneName=dyn_front_poncho_05}
    constraintSetup [{targetBone {boneName=pelvis}}]
    constraintWeights [1,0]
   }
   Save cached pose 'front_constraint_cache'(Pose=Component To Local(ComponentPose=@n19)) {cacheName=front_constraint_cache}
   n20 = "Transform (Modify) Bone - Bone: dyn_back_poncho_01"!(
    ComponentPose=Local To Component(LocalPose=Use cached pose 'copypose_cache'())
    bAlphaBoolEnabled:bool=True
    Translation="0,1,0"
    Rotation="0,0,0"
    Scale="1,1,1"
   ) -> (Pose) {
    alphaInputType=Bool
    translationMode=BMM_Additive
    translationSpace=BCS_BoneSpace
    alphaBoolBlend {blendInTime=0.1, blendOption=Linear, blendOutTime=0.3}
    boneToModify {boneName=dyn_back_poncho_01}
    translation {y=1}
   }
   n21 = RigidBody(ComponentPose=@n20, ExternalForce=Get skydiving_force() -> (skydiving_force:Vector)) {
    bClampLinearTranslationLimitToRefPose=true
    bOverrideWorldGravity=true
    overridePhysicsAsset=&/F_MED_BountyHunter_BackPoncho_Physics
    simulationSpace=BaseBoneSpace
    alphaBoolBlend {blendOption=Linear}
    baseBoneRef {boneName=spine_02}
    componentAppliedLinearAccClamp=<3000 3000 3000>
    componentLinearAccScale=<0.3 0.3 0.3>
    componentLinearVelScale=<0.3 0.3 0.3>
    overrideWorldGravity {z=-980}
    simSpaceSettings {
     maxAngularAcceleration=200
     maxAngularVelocity=5
     maxLinearAcceleration=1000
     maxLinearVelocity=300
     velocityScaleZ=0
     worldAlpha=1
     externalLinearDragV=<1 1>
    }
   }
   Save cached pose 'back_dynamics_cache'(Pose=Component To Local(ComponentPose=@n21)) {cacheName=back_dynamics_cache}
   n22 = RigidBody(ComponentPose=Local To Component(LocalPose=Use cached pose 'copypose_cache'()), ExternalForce=Get skydiving_force()) {
    bOverrideWorldGravity=true
    overridePhysicsAsset=&/F_MED_BountyHunter_FrontPoncho_Physics
    simulationSpace=BaseBoneSpace
    alphaBoolBlend {blendOption=Linear}
    baseBoneRef {boneName=spine_01}
    componentLinearAccScale=<0.5 0.5 0.5>
    componentLinearVelScale=<0.5 0.5 0.5>
    overrideWorldGravity {z=-980}
   }
   Save cached pose 'front_dynamics_cache'(Pose=Component To Local(ComponentPose=@n22)) {cacheName=front_dynamics_cache}
   Save cached pose 'back_cache'(Pose=Blend Poses by bool(
     BlendPose_0=Use cached pose 'back_constraint_cache'()
     BlendPose_1=Use cached pose 'back_dynamics_cache'()
     BlendTime_0=0
     BlendTime_1=0
     bActiveValue=Get isBackStatic() -> (isBackStatic:bool)
    ) {blendType=Linear}) {cacheName=back_cache}
   Save cached pose 'front_cache'(Pose=Blend Poses by bool(
     BlendPose_0=Use cached pose 'front_constraint_cache'()
     BlendPose_1=Use cached pose 'front_dynamics_cache'()
     BlendTime_0=0
     BlendTime_1=0
     bActiveValue=Get isRigidBodyDisabled() -> (isRigidBodyDisabled:bool)
    ) {blendType=Linear}) {cacheName=front_cache}
   Get bIsPlayingMeleeAnim()
  ]}
 EventGraph {nodes [
   Event BlueprintUpdateAnimation()
   > n1 = Sequence()
   n2 = "integer > integer"(A=Get CurrentLOD(), B=1) {is=title K2Node_CallFunction}
   n3 = OR Boolean(A=NOT Boolean(A=Get bRigidBodyEnabled()), B=@n2)
   > Set isRigidBodyDisabled(isRigidBodyDisabled:bool=@n3) -> (Output_Get:bool)
   > Set isBackStatic(isBackStatic:bool=OR Boolean(A=Get isBackpackEquipped() -> (isBackpackEquipped:bool), B=@n3)) -> (Output_Get:bool)
   > Set back_poncho_lod_offset(back_poncho_lod_offset:Vector=SelectVector(A="0,-5, 0", B="0, 0, 0", bPickA=@n2)) -> (Output_Get:Vector)
   > Set back_poncho_lod_rot_offset(back_poncho_lod_rot_offset:Rotator=SelectRotator!(A:Rotator="0,20, 0", B:Rotator="0, 0, 0", bPickA:bool=@n2) -> (ReturnValue:Rotator)) -> (Output_Get:Rotator)
   n4 = Get SkydivingExternalForceInput()
   Set skydiving_force(
    execute=@n1.then_1
    skydiving_force:Vector=vector * float(
     A=Break Fort Anim Input Skydiving External Force(FortAnimInput_SkydivingExternalForce=@n4) -> (
      bUseSkydivingVectorForce=~
      SkydivingVectorForce=~
      HeadToPelvisDirection=~
      FloatingMultiplier=~
      FloatingAdditive=~
      DivingMultiplier=~
      DivingAdditive=~
      ParachutingMultiplier=~
      ParachutingAdditive=~
     ).SkydivingVectorForce
     B=0.4
    )
   ) -> (Output_Get:Vector)
   Event BlueprintInitializeAnimation()
   n5 = TryGetPawnOwner()
   > Is Valid(InputObject=@n5)
   > n6 = Cast To FortPlayerPawnAthena(Object=@n5)
   > Set player_pawn(player_pawn:Fort Player Pawn Athena Object Reference=@n6) -> (Output_Get:Fort Player Pawn Athena Object Reference)
   Event OnCharacterCustomizationLoaded()
   n7 = Get player_pawn() -> (player_pawn:Fort Player Pawn Athena Object Reference)
   > Is Valid(InputObject=@n7)
   > Set isBackpackEquipped(isBackpackEquipped:bool="Not Equal (Object)"!(A:Object Reference=GetSkeletalMeshForPartType(self=@n7, PartType=Backpack), B:Object Reference) -> (ReturnValue:bool)) -> (Output_Get:bool)
  ]}
]
variables [
 back_poncho_lod_offset {type=Vector}
 back_poncho_lod_rot_offset {type=Rotator}
 isBackpackEquipped {type=boolean}
 isBackStatic {type=boolean}
 isRigidBodyDisabled {type=boolean}
 player_pawn {type=FortPlayerPawnAthena}
 skydiving_force {type=Vector}
]
